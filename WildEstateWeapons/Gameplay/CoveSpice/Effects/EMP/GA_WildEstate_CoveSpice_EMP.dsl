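; Event-triggered gameplay ability for the WildEstate CoveSpice weapon that applies an EMP effect to a player pawn.
; On activation it stores the event data, casts an actor from it to FortPlayerPawn, and applies the EMP GE to that pawn.
; A scalable-float delay holds the ability open before EndAbility, as a temporary workaround for a noted retriggering stacking issue.
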(event Ability|EventActivateAbilityFromEvent
  (params (EventData :type Gameplay%20Event%20Data%20Structure%20%28by%20ref%29))
  (|CallParentFunction EventData)
  (bind _output_get (Variables|Default|SetEvent%20Data EventData))
  (bind _asfort_player_pawn (Utilities|Casting|CastToFortPlayerPawn (Utilities|Struct|BreakGameplayEventData _output_get))
    (:then
      (Ability|ApplyGameplayEffectToTarget (Ability|TargetData|AbilityTargetDatafromActor _asfort_player_pawn) (Variables|Default|GetEMP%20GE))
      ;[0.15 0.004276 0 0.5] Remove this after fixing the retriggering stacking issue
      (Utilities|FlowControl|Delay (ScalableFloat|GetValueatLevel (Variables|Default|GetTemp%20Delay%20Duration))))
    (:CastFailed
      _))
  (Ability|EndAbility))
